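# Ground material for island runes that blends world-projected corrupted rocky ground with UV-mapped rune textures by vertex color red.
# The corruption layer tiles on world XY, with cheaper normal, specular and roughness paths for lower quality levels or mobile.
# The rune layer's switchable emissive mixes two tinted colors through a world-space panning storm-swirl mask, then confines the glow with a rune mask texture.
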Material /Figment_S06/Athena/Environments/Island/Runes/Material/M_RuneGround_02_Emissive
graph {
 properties {bUseMaterialAttributes=true, pixelDepthOffsetMode=PDOM_Legacy, usedShadingModels=Default Lit}
 nodes [
  n1 = Divide(
   A=ComponentMask(None=WorldPosition().XYZ) {g=true, r=true}
   B=ScalarParameter() {defaultValue=1536, group=Corruption, parameterName=Corruption_TexScale_1}
  )
  n2 = TextureSampleParameter2D(UVs=@n1) {
   group=Corruption
   parameterName=Corruption_BaseColor
   texture=/Figment_S06/Maps/Test_Maps/Environments/StormRavaged/Terrain/Textures/T_Corruptground_Rocky2_D
  }
  n3 = Multiply(A=@n2.RGB, B=VectorParameter() {group=Corruption, parameterName=Corruption_BaseTint, defaultValue=<rgba 1 1 1 0>}.RGB)
  MakeMaterialAttributes!(
   BaseColor=@n3
   Metallic
   Specular=Constant()
   Roughness
   Anisotropy
   EmissiveColor
   Opacity
   OpacityMask
   Normal
   Tangent
   WorldPositionOffset
   SubsurfaceColor
   ClearCoat
   ClearCoatRoughness
   AmbientOcclusion
   Refraction
   CustomizedUVs_0
   CustomizedUVs_1
   CustomizedUVs_2
   CustomizedUVs_3
   CustomizedUVs_4
   CustomizedUVs_5
   CustomizedUVs_6
   CustomizedUVs_7
   PixelDepthOffset
   ShadingModel
   Displacement
  )
  n4 = TextureSampleParameter2D() {parameterName=Normal, samplerType=SAMPLERTYPE_Normal, texture=/Game/Athena/Environments/Island/Runes/Texture/T_RuneGround_02_Emissive_N}
  MaterialFunctionCall(BaseNormal=@n4.RGB, AdditionalNormal=Constant3Vector() {constant=<rgba 0 0 1 0>}) {materialFunction=&BlendAngleCorrectedNormals}
  n5 = Constant()
  n6 = ShadingPathSwitch(
   Default=Multiply(A=@n2.R, B=ScalarParameter() {defaultValue=0.9, group=Corruption, parameterName=Corruption_SpecGain})
   Mobile=Multiply(A=@n2.R, B=ScalarParameter() {defaultValue=1, group=Corruption, parameterName=SimpleMobileSpec})
  ) {desc=Migrated from FeatureLevelSwitch}
  n7 = Constant() {r=1}
  n8 = ShadingPathSwitch(
   Default=LinearInterpolate(
    A=ScalarParameter() {defaultValue=0.77, group=Corruption, parameterName=Corruption_RMax}
    B=ScalarParameter() {defaultValue=-5, group=Corruption, parameterName=Corruption_RMin}
    Alpha=@n2.R
   )
   Forward=@n7
   Mobile=@n7
  )
  n9 = TextureSampleParameter2D(UVs=@n1) {group=Corruption, parameterName=Corruption_Normal, samplerType=SAMPLERTYPE_Normal, texture=/Game/Maps/Environments/T_Corruptground_Rocky_N}
  n10 = VertexNormalWS()
  n11 = QualitySwitch(Default=@n9.RGB, Low=@n10, High=@n9.RGB, Medium=@n10, Epic=@n9.RGB)
  n12 = ScalarParameter() {defaultValue=1, parameterName=NormalXY}
  n13 = Multiply(
   A=Transform(None=ShadingPathSwitch(Default=@n11, Deferred=@n11, Mobile=@n10) {desc=Migrated from FeatureLevelSwitch}) {transformSourceType=TRANSFORMSOURCE_World, transformType=TRANSFORM_Tangent}
   B=MaterialFunctionCall(X=@n12, Y=@n12, Z=ScalarParameter() {defaultValue=1, parameterName=NormalZ}) {materialFunction=&MakeFloat3}.Result
  )
  n14 = TextureSampleParameter2D() {parameterName=Diffuse, texture=/Game/Athena/Environments/Island/Runes/Texture/T_RuneGround_02_Emissive_D}
  n15 = TextureSampleParameter2D() {parameterName=Mask, samplerType=SAMPLERTYPE_Masks, texture=/Game/Athena/Environments/Island/Runes/Texture/T_RuneGround_02_Emissive_S}
  n16 = Multiply(A=@n15.R, B=ScalarParameter() {defaultValue=0.9, group=Corruption, parameterName=Corruption_SpecGainRocks})
  n17 = Multiply(A=@n15.B, B=ScalarParameter() {defaultValue=0.9, group=Corruption, parameterName=Corruption_RoughnessGainRocks})
  n18 = ComponentMask(None=WorldPosition().XYZ) {g=true, r=true}
  n19 = ScalarParameter() {defaultValue=3000, group=Corruption, parameterName=EmissivePanner}
  n20 = TextureSample(UVs=Panner(Coordinate=Divide(A=@n18, B=@n19), Time=Time(), Speed=ScalarParameter() {defaultValue=0.04, parameterName=PanSpeed}) {speedX=0.015, speedY=-0.02}) {samplerType=SAMPLERTYPE_LinearGrayscale, texture=/Figment_S06/VisualThreatManager/StormVisuals/LargeClouds/T_StormSwirlMask}
  n21 = LinearInterpolate(
   A=Multiply(
    A=VectorParameter() {parameterName=EmissiveColor, defaultValue=<rgba 1 0.0103 1 1>}.RGB
    B=ScalarParameter() {defaultValue=8, parameterName=EmissivePower}
   )
   B=Multiply(
    A=VectorParameter() {parameterName=EmissiveColor_1, defaultValue=<rgba 0.5518 0.1107 0.6979 0>}.RGB
    B=ScalarParameter() {defaultValue=5, parameterName=EmissivePower_1}
   )
   Alpha=@n20.RGB
  ) {constB=0}
  n22 = StaticSwitchParameter(
   True=LinearInterpolate(
    B=LinearInterpolate(B=@n21, Alpha=@n20.RGB)
    Alpha=TextureSample() {samplerType=SAMPLERTYPE_Masks, texture=/Figment_S06/Athena/Environments/Island/Runes/Texture/T_RuneGround_02_Emissive_E}.R
   )
   False=Constant()
  ) {defaultValue=true, parameterName=UseEmissive}
  MaterialFunctionCall!(
   Material=BlendMaterialAttributes(
    A=MakeMaterialAttributes(BaseColor=@n3, Metallic=@n5, Specular=@n6, Roughness=@n8, Normal=@n13)
    B=MakeMaterialAttributes(BaseColor=@n14.RGB, Specular=@n16, Roughness=@n17, EmissiveColor=@n22, Normal=@n4.RGB)
    Alpha=VertexColor().R
   )
   Normal
  ) {materialFunction=/Engine/Functions/MaterialLayerFunctions/MatLayerBlend_BakedNormal}
  TextureCoordinate()
 ]
}
